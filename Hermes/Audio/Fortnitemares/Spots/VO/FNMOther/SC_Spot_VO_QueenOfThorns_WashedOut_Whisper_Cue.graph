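# Sound cue for washed-out, atmospheric Queen of Thorns whisper VO in Fortnitemares, set up as a distant outdoor ambient one-shot.
# An equally weighted Random node picks one of 21 voice waves and feeds a Modulator that slightly varies pitch and volume.
# The quality-level root routes its first two slots to that same chain and leaves the other two empty.
# Distant-creature attenuation, boosted volume and a two-voice concurrency override shape playback.
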SoundCue /Hermes/Audio/Fortnitemares/Spots/VO/FNMOther/SC_Spot_VO_QueenOfThorns_WashedOut_Whisper_Cue
root=QualityLevel_0
nodes [
 {id=Modulator_3, children [Random_3], properties {pitchMax=1.1, pitchMin=0.9, volumeMax=1.1, volumeMin=1}}
 {id=QualityLevel_0, children [Modulator_3,Modulator_3,null,null]}
 {
  id=Random_3
  children [
   WavePlayer_3
   WavePlayer_4
   WavePlayer_5
   WavePlayer_6
   WavePlayer_7
   WavePlayer_8
   WavePlayer_9
   WavePlayer_13
   WavePlayer_14
   WavePlayer_15
   WavePlayer_16
   WavePlayer_17
   WavePlayer_18
   WavePlayer_19
   WavePlayer_20
   WavePlayer_21
   WavePlayer_22
   WavePlayer_23
   WavePlayer_24
   WavePlayer_25
   WavePlayer_26
  ]
  properties {weights [1,1,1,1,1,1,1,1,1,1,1,1,1,1,1,1,1,1,1,1,1]}
 }
 {
  id=WavePlayer_13
  properties {soundWaveAssetPtr=&/VO_TangyRadish_QueenOfThorns_Atmospheric_WashedOut_090}
 }
 {
  id=WavePlayer_14
  properties {soundWaveAssetPtr=&/VO_TangyRadish_QueenOfThorns_Atmospheric_WashedOut_083}
 }
 {
  id=WavePlayer_15
  properties {soundWaveAssetPtr=&/VO_TangyRadish_QueenOfThorns_Atmospheric_WashedOut_082}
 }
 {
  id=WavePlayer_16
  properties {soundWaveAssetPtr=&/VO_TangyRadish_QueenOfThorns_Atmospheric_WashedOut_081}
 }
 {
  id=WavePlayer_17
  properties {soundWaveAssetPtr=&/VO_TangyRadish_QueenOfThorns_Atmospheric_WashedOut_080}
 }
 {
  id=WavePlayer_18
  properties {soundWaveAssetPtr=&/VO_TangyRadish_QueenOfThorns_Atmospheric_WashedOut_061}
 }
 {
  id=WavePlayer_19
  properties {soundWaveAssetPtr=&/VO_TangyRadish_QueenOfThorns_Atmospheric_WashedOut_051}
 }
 {
  id=WavePlayer_20
  properties {soundWaveAssetPtr=&/VO_TangyRadish_QueenOfThorns_Atmospheric_WashedOut_043}
 }
 {
  id=WavePlayer_21
  properties {soundWaveAssetPtr=&/VO_TangyRadish_QueenOfThorns_Atmospheric_WashedOut_041}
 }
 {
  id=WavePlayer_22
  properties {soundWaveAssetPtr=&/VO_TangyRadish_QueenOfThorns_Atmospheric_WashedOut_035}
 }
 {
  id=WavePlayer_23
  properties {soundWaveAssetPtr=&/VO_TangyRadish_QueenOfThorns_Atmospheric_WashedOut_030}
 }
 {
  id=WavePlayer_24
  properties {soundWaveAssetPtr=&/VO_TangyRadish_QueenOfThorns_Atmospheric_WashedOut_022}
 }
 {
  id=WavePlayer_25
  properties {soundWaveAssetPtr=&/VO_TangyRadish_QueenOfThorns_Atmospheric_WashedOut_011}
 }
 {
  id=WavePlayer_26
  properties {soundWaveAssetPtr=&/VO_TangyRadish_QueenOfThorns_Atmospheric_WashedOut_010}
 }
 {
  id=WavePlayer_3
  properties {soundWaveAssetPtr=&/VO_TangyRadish_QueenOfThorns_Atmospheric_WashedOut_110}
 }
 {
  id=WavePlayer_4
  properties {soundWaveAssetPtr=&/VO_TangyRadish_QueenOfThorns_Atmospheric_WashedOut_070}
 }
 {
  id=WavePlayer_5
  properties {soundWaveAssetPtr=&/VO_TangyRadish_QueenOfThorns_Atmospheric_WashedOut_012}
 }
 {
  id=WavePlayer_6
  properties {soundWaveAssetPtr=&/VO_TangyRadish_QueenOfThorns_Atmospheric_WashedOut_122}
 }
 {
  id=WavePlayer_7
  properties {soundWaveAssetPtr=&/VO_TangyRadish_QueenOfThorns_Atmospheric_WashedOut_112}
 }
 {
  id=WavePlayer_8
  properties {soundWaveAssetPtr=&/VO_TangyRadish_QueenOfThorns_Atmospheric_WashedOut_104}
 }
 {
  id=WavePlayer_9
  properties {soundWaveAssetPtr=&/VO_TangyRadish_QueenOfThorns_Atmospheric_WashedOut_101}
 }
]
properties {
 attenuationSettings=/Game/Sounds/Attenuation/Fort_TileAmbient_OneShot_Distant_Creature_Att
 bOverrideConcurrency=true
 maxDistance=8500
 soundClassObject=&Fort_Ambient_OneShots_Outdoors
 volumeMultiplier=2
 concurrencyOverrides {maxCount=2}
}
